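# Niagara swing-trail effect for the AnvilAlarmMoth DrumStick pickaxe, drawing two short-lived ribbons along the swing.
# Accent_trail and Trail both use shared FN anim-trail modules to spawn and update ribbons, with width curves and dynamic material parameters active only on the accent.
# Both set TopLineColor from a system time-of-day tint and variant-indexed color curves, with alpha shaped by age, time of day and fade, plus a front-end modifier on Trail.
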NiagaraSystem /SparksCosmetics/Instrument/DrumStick/AnvilAlarmMoth/Effects/NS_Pickaxe_DrumStick_AnvilAlarmMoth_SwingTrail
dependencies {
 usedDataInterfaces [&NiagaraDataInterfaceCurve, &NiagaraDataInterfaceColorCurve, &NiagaraDataInterfaceSocketReader]
 usedDynamicInputs [&MakeIntFromFloat, &MergeHelper_FloatSwitch]
 usedModules [
  &self:SystemSpawnScript.SystemScriptSource.SystemScriptGraph.NiagaraNodeAssignment_2.SetVariables_#1
  &self:SystemSpawnScript.SystemScriptSource.SystemScriptGraph.NiagaraNodeAssignment_3.SetVariables_#2
  &self:SystemSpawnScript.SystemScriptSource.SystemScriptGraph.NiagaraNodeAssignment_4.SetVariables_#3
  &SystemState
  &self:SystemSpawnScript.SystemScriptSource.SystemScriptGraph.NiagaraNodeAssignment_1.SetVariables_#4
  &self:SystemSpawnScript.SystemScriptSource.SystemScriptGraph.NiagaraNodeAssignment_5.SetVariables_#5
  &self:SystemSpawnScript.SystemScriptSource.SystemScriptGraph.NiagaraNodeAssignment_6.SetVariables_#6
 ]
}
emitters {
 Accent_trail {
  emitterUpdateScript [
   &EmitterState()
   &FN_SpawnAnimTrail()
  ]
  particleSpawnScript [
   &FN_InitializeAnimTrailBoneIndecies(First Bone or Socket Index=1) {module=FN_InitializeAnimTrailBoneIndecies001}
   &FN_InitializeAnimTrail(AnimTrailWidth=7, Lifetime=0.15) {module=FN_InitializeAnimTrail001}
   &AddVelocity(Velocity=<0 500 0>) {enabled=false}
  ]
  particleUpdateScript [
   &ScaleRibbonWidth(Ribbon Width Scale=&FloatFromCurve(FloatCurve=DataInterface() {propertyValues {Curve {keys [
         {arriveTangent=7.6644, interpMode=RCIM_Cubic, leaveTangent=7.6644, tangentMode=RCTM_User, time=0, value=0.5}
         {arriveTangentWeight=0.0745, interpMode=RCIM_Cubic, leaveTangentWeight=0.2981, tangentMode=RCTM_User, time=0.2, value=1}
         {arriveTangent=-2.1703, arriveTangentWeight=0.3708, interpMode=RCIM_Cubic, leaveTangent=-2.1703, tangentMode=RCTM_User, time=1, value=0}
        ]}}}))
   &FN_UpdateAnimTrail(Lock Particles=true) {module=FN_UpdateAnimTrail001}
   &SolveForcesAndVelocity(Clamp Velocity=false, Limit Acceleration=false)
   &DynamicMaterialParameters(
    Index 0 Param 1=0.5
    Index 1 Param 3=&FloatFromCurve(
     FloatCurve=DataInterface() {propertyValues {Curve {defaultValue=0, keys [{tangentMode=RCTM_User, time=0, value=1}, {tangentMode=RCTM_User, time=1, value=0}]}}}
     CurveIndex=&Frac_Float(Float=&Multiply_Float(A=$Engine.Time, B=1))
    )
    Index 1 Param 4=1
    Index 2 Param 1=5
    Index 2 Param 2=2
    Module.Index0Float0Write=true
    Module.Index0Float1Write=true
    Module.Index0Float2Write=true
    Module.Index0Float3Write=true
    Module.Index1Float0Write=true
    Module.Index1Float1Write=true
    Module.Index1Float2Write=true
    Module.Index1Float3Write=true
    Module.Index2Float0Write=true
    Module.Index2Float1Write=true
    Module.Index2Float2Write=true
    Module.Index2Float3Write=true
    Write Parameter Index 1=::NewEnumerator1
    Write Parameter Index 2=::NewEnumerator1
   )
   /SetVariables(Particles.TopLineColor:LinearColor=&MakeLinearColorFromVectorAndFloat(
     "Vector (RGB)"=&Multiply_Vector(
      A=&MakeVectorFromLinearColorRGB(Color=$System.TOD Color)
      B=&MergeHelper_VectorSwitch(
       0=&MakeVectorFromLinearColorRGB(Color=&ColorFromCurve(ColorCurve=DataInterface() {propertyValues {
           AlphaCurve {keys [<at 0 1 RCIM_Cubic>]}
           BlueCurve {keys [<at 0.7117 2 RCIM_Cubic>]}
           GreenCurve {keys [<at 0.7117 2 RCIM_Cubic>]}
           RedCurve {keys [<at 0.7117 2 RCIM_Cubic>]}
          }}))
       1=&MakeVectorFromLinearColorRGB(Color=&ColorFromCurve(ColorCurve=DataInterface() {propertyValues {
           AlphaCurve {keys [<at 0 1 RCIM_Cubic>]}
           BlueCurve {keys [<at 0 0 RCIM_Cubic>]}
           GreenCurve {keys [<at 0 0 RCIM_Cubic>]}
           RedCurve {keys [<at 0 0 RCIM_Cubic>]}
          }}))
       2=?
       3=?
       Index=$System.VariantInt
       Num=::NewEnumerator0
      )
     )
     "Float (Alpha)"=&FloatFromCurve(
      FloatCurve=DataInterface() {propertyValues {Curve {keys [
          {arriveTangent=9.9264, interpMode=RCIM_Cubic, leaveTangent=9.9264, tangentMode=RCTM_User, time=0, value=0.3}
          {interpMode=RCIM_Cubic, leaveTangent=-0.962, tangentMode=RCTM_Break, time=0.1, value=0.8}
          {arriveTangent=-0.5809, interpMode=RCIM_Cubic, leaveTangent=-0.5809, tangentMode=RCTM_User, time=0.5, value=0.15}
          {arriveTangent=-0.2139, interpMode=RCIM_Cubic, tangentMode=RCTM_Break, time=1, value=0}
         ]}}}
      Scale Curve=&FloatFromCurve(
       FloatCurve=DataInterface() {propertyValues {Curve {keys [
           <at 0 0.88 RCIM_Cubic>
           {arriveTangentWeight=0.0858, interpMode=RCIM_Cubic, leaveTangentWeight=0.0558, tangentMode=RCTM_User, time=0.4434, value=1}
           {arriveTangentWeight=0.0572, interpMode=RCIM_Cubic, leaveTangentWeight=0.1462, tangentMode=RCTM_User, time=0.6, value=1}
           <at 1 0.88 RCIM_Cubic>
          ]}}}
       CurveIndex=Linked() {linkedVariable {name=NPC.FortniteNPC.TimeOfDay_0-1, type=NiagaraFloat}}
       Scale Curve=$System.AlphaFade
      )
     )
    )) {isSetParameters=true}
  ]
  rendererClasses [&NiagaraRibbonRendererProperties]
  renderers [{rendererClass=&NiagaraRibbonRendererProperties}]
 }
 Trail {
  emitterSpawnScript [&FN_RibbonIDOffsetUtility()]
  emitterUpdateScript [
   &EmitterState()
   &FN_SpawnAnimTrail(AnimTrailWidthMode=::FromFirst)
  ]
  particleSpawnScript [
   &FN_InitializeAnimTrailBoneIndecies()
   &FN_InitializeAnimTrail(Lifetime=0.15) {module=FN_InitializeAnimTrail001}
   &AddVelocity(Velocity=<0 500 0>) {enabled=false}
  ]
  particleUpdateScript [
   &SolveForcesAndVelocity(Clamp Velocity=false, Limit Acceleration=false)
   &FN_UpdateAnimTrail()
   &DynamicMaterialParameters(
    Index 0 Param 1=1
    Index 0 Param 3=0.1
    Module.Index0Param0Write=true
    Module.Index0Param1Write=true
    Module.Index0Param2Write=true
    Module.Index0Param3Write=true
    Module.Index1Param0Write=true
    Module.Index1Param1Write=true
    Module.Index1Param2Write=true
    Module.Index1Param3Write=true
    Write Parameter Index 1=true
   ) {enabled=false}
   /SetVariables(Particles.TopLineColor:LinearColor=&MakeLinearColorFromVectorAndFloat(
     "Vector (RGB)"=&Multiply_Vector(
      A=&MakeVectorFromLinearColorRGB(Color=$System.TOD Color)
      B=&MakeVectorFromLinearColorRGB(Color=&ColorFromCurve(ColorCurve=DataInterface() {propertyValues {
          AlphaCurve {keys [<at 0 1 RCIM_Cubic>]}
          BlueCurve {keys [<at 0 1 RCIM_Cubic>,<at 0.5861 0.5664 RCIM_Cubic>]}
          GreenCurve {keys [<at 0 0.4964 RCIM_Cubic>,<at 0.5861 0.2146 RCIM_Cubic>]}
          RedCurve {keys [<at 0 0.2455 RCIM_Cubic>,<at 0.5861 0.3251 RCIM_Cubic>]}
         }}))
     )
     "Float (Alpha)"=&Multiply_Float(
      A=$System.AlphaFade
      B=&FloatFromCurve(
       FloatCurve=DataInterface() {propertyValues {Curve {keys [
           {interpMode=RCIM_Cubic, leaveTangent=-1.8267, leaveTangentWeight=0.5724, tangentMode=RCTM_Break, time=0, value=1}
           {arriveTangent=-0.2139, interpMode=RCIM_Cubic, tangentMode=RCTM_Break, time=0.9, value=0}
          ]}}}
       Scale Curve=&FloatFromCurve(
        FloatCurve=DataInterface() {propertyValues {Curve {keys [
            <at 0.1 0.8 RCIM_Cubic>
            {arriveTangentWeight=0.0858, interpMode=RCIM_Cubic, leaveTangentWeight=0.0558, tangentMode=RCTM_User, time=0.4434, value=1}
            {arriveTangentWeight=0.0572, interpMode=RCIM_Cubic, leaveTangentWeight=0.1462, tangentMode=RCTM_User, time=0.6, value=1}
            <at 0.9 0.8 RCIM_Cubic>
           ]}}}
        CurveIndex=Linked() {linkedVariable {name=NPC.FortniteNPC.TimeOfDay_0-1, type=NiagaraFloat}}
        Scale Curve=&FN_FrontEndFloatModifier(Gameplay=0.25, Front End=0.15)
       )
      )
     )
    )) {enabled=false, isSetParameters=true}
   /SetVariables(Particles.TopLineColor:LinearColor=&MakeLinearColorFromVectorAndFloat(
     "Vector (RGB)"=&Multiply_Vector(
      A=&MakeVectorFromLinearColorRGB(Color=$System.TOD Color)
      B=&MergeHelper_VectorSwitch(
       0=&MakeVectorFromLinearColorRGB(Color=&ColorFromCurve(ColorCurve=DataInterface() {propertyValues {
           AlphaCurve {keys [<at 0 1 RCIM_Cubic>]}
           BlueCurve {keys [<at 0.7117 0 RCIM_Cubic>]}
           GreenCurve {keys [<at 0.7117 0 RCIM_Cubic>]}
           RedCurve {keys [<at 0.7117 0 RCIM_Cubic>]}
          }}))
       1=&MakeVectorFromLinearColorRGB(Color=&ColorFromCurve(ColorCurve=DataInterface() {propertyValues {
           AlphaCurve {keys [<at 0 1 RCIM_Cubic>]}
           BlueCurve {keys [<at 0.6648 0>, {arriveTangent=2.5607, interpMode=RCIM_Cubic, time=0.8921, value=1}, <at 1 0.5664 RCIM_Cubic>]}
           GreenCurve {keys [<at 0.6648 0>, {arriveTangent=0.9703, interpMode=RCIM_Cubic, time=0.8921, value=0.4964}, <at 1 0.2146 RCIM_Cubic>]}
           RedCurve {keys [<at 0.6648 0>, {arriveTangent=1.4699, interpMode=RCIM_Cubic, time=0.8921, value=0.2455}, <at 1 0.3251 RCIM_Cubic>]}
          }}))
       2=?
       3=?
       Index=$System.VariantInt
       Num=::NewEnumerator0
      )
     )
     "Float (Alpha)"=&Multiply_Float(
      A=$System.AlphaFade
      B=&FloatFromCurve(
       FloatCurve=DataInterface() {propertyValues {Curve {keys [
           {interpMode=RCIM_Cubic, leaveTangent=-1.8267, leaveTangentWeight=0.5724, tangentMode=RCTM_Break, time=0, value=1}
           {arriveTangent=-0.2139, interpMode=RCIM_Cubic, tangentMode=RCTM_Break, time=0.9, value=0}
          ]}}}
       Scale Curve=&FloatFromCurve(
        FloatCurve=DataInterface() {propertyValues {Curve {keys [
            <at 0.1 0.8 RCIM_Cubic>
            {arriveTangentWeight=0.0858, interpMode=RCIM_Cubic, leaveTangentWeight=0.0558, tangentMode=RCTM_User, time=0.4434, value=1}
            {arriveTangentWeight=0.0572, interpMode=RCIM_Cubic, leaveTangentWeight=0.1462, tangentMode=RCTM_User, time=0.6, value=1}
            <at 0.9 0.8 RCIM_Cubic>
           ]}}}
        CurveIndex=Linked() {linkedVariable {name=NPC.FortniteNPC.TimeOfDay_0-1, type=NiagaraFloat}}
        Scale Curve=&FN_FrontEndFloatModifier(Gameplay=0.35, Front End=0.35)
       )
      )
     )
    )) {isSetParameters=true}
   &ScaleRibbonWidth(Ribbon Width Scale=&FloatFromCurve(FloatCurve=DataInterface() {propertyValues {Curve {
        defaultValue=1
        keys [
         {interpMode=RCIM_Cubic, tangentMode=RCTM_User, time=0, value=1}
         {arriveTangent=-3.0576, interpMode=RCIM_Cubic, leaveTangent=-3.0576, tangentMode=RCTM_User, time=1, value=0}
        ]
       }}})) {enabled=false}
  ]
  rendererClasses [&NiagaraRibbonRendererProperties]
  renderers [{rendererClass=&NiagaraRibbonRendererProperties}]
 }
}
summary {
 systemName=NS_Pickaxe_DrumStick_AnvilAlarmMoth_SwingTrail
 emitters [
  {emitterName=Accent_trail, rendererClasses [&NiagaraRibbonRendererProperties]}
  {emitterName=Trail, rendererClasses [&NiagaraRibbonRendererProperties]}
 ]
 userVariables [
  User.bIsFrontEnd {type=NiagaraBool}
  User.bIsFrontEndPreview {type=NiagaraBool}
  User.RibbonIDOffset {type=NiagaraInt32}
  User.SkeletalMesh {
   type=NiagaraDataInterfaceSkeletalMesh
   defaultValue {
    struct=NiagaraExt_VariableValue_DataInterface
    value {dataInterface=&self:NiagaraDataInterfaceSkeletalMesh_0}
   }
  }
  User.SwingSectionIndex {type=NiagaraInt32}
  User.VariantID {type=NiagaraFloat, defaultValue=0, description=""}
 ]
}
