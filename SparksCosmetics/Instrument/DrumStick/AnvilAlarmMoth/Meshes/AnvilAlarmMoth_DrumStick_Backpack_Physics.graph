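# Physics asset for the backpack mesh of the AnvilAlarmMoth drumstick cosmetic in SparksCosmetics, covering its left and right dangling chains.
# Each three-bone dyn_chain carries one sphere body per bone, with a much larger sphere on the end bone.
# Constraints link each bone to the one before it with equal swing and twist limits.
# No constraint ties either first link to a parent body or joins the two sides.
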PhysicsAsset /SparksCosmetics/Instrument/DrumStick/AnvilAlarmMoth/Meshes/AnvilAlarmMoth_DrumStick_Backpack_Physics
bodies {
 dyn_chain_01_l [{shapeType=Sphere, radius=1.5164, shapeName=dyn_chain_01_l_sphere, center=<1.2805 -0.0004 -0.0006>}]
 dyn_chain_01_r [{shapeType=Sphere, radius=1.5333, shapeName=dyn_chain_01_r_sphere, center=<1.2735 0.0001 0.0445>}]
 dyn_chain_02_l [{shapeType=Sphere, radius=1.0582, shapeName=dyn_chain_02_l_sphere, center=<0.3594 -0.003 0>}]
 dyn_chain_02_r [{shapeType=Sphere, radius=1.0385, shapeName=dyn_chain_02_r_sphere, center=<0.3337 0.0066 -0.0271>}]
 dyn_chain_03_l [{shapeType=Sphere, radius=3.9892, shapeName=dyn_chain_03_l_sphere, center=<2.1251 0.0004 -0.0004>}]
 dyn_chain_03_r [{shapeType=Sphere, radius=3.9982, shapeName=dyn_chain_03_r_sphere, center=<2.1156 0.0084 -0.0505>}]
}
constraints [
 {bone1Name=dyn_chain_02_l, bone2Name=dyn_chain_01_l, swing1LimitDegrees=45, swing2LimitDegrees=45, twistLimitDegrees=45}
 {bone1Name=dyn_chain_03_l, bone2Name=dyn_chain_02_l, swing1LimitDegrees=45, swing2LimitDegrees=45, twistLimitDegrees=45}
 {bone1Name=dyn_chain_02_r, bone2Name=dyn_chain_01_r, swing1LimitDegrees=45, swing2LimitDegrees=45, twistLimitDegrees=45}
 {bone1Name=dyn_chain_03_r, bone2Name=dyn_chain_02_r, swing1LimitDegrees=45, swing2LimitDegrees=45, twistLimitDegrees=45}
]
