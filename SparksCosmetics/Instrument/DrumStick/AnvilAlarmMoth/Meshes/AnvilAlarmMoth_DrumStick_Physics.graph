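# Physics asset for the AnvilAlarmMoth drumstick, a Sparks instrument cosmetic, giving collision bodies and joints to its handle and dynamic chain.
# One long capsule covers the handle, and smaller capsules cover the three dyn_chain bones.
# Constraints link the chain in series from the handle, with matching limited swing on both axes and very tight twist limits, so it can bend with little axial rotation.
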PhysicsAsset /SparksCosmetics/Instrument/DrumStick/AnvilAlarmMoth/Meshes/AnvilAlarmMoth_DrumStick_Physics
bodies {
 dyn_chain_01 [{shapeType=Capsule, length=0, radius=1.5199, shapeName=dyn_chain_01_capsule, center=<1.272 0 0>, rotation=<rot -0.0085 -90 0.0003>}]
 dyn_chain_02 [{shapeType=Capsule, length=0.0301, radius=1.0301, shapeName=dyn_chain_02_capsule, center=<0.3336 0 0.0036>, rotation=<rot -0.022 -89.9997 0>}]
 dyn_chain_03 [{shapeType=Capsule, length=4.0037, radius=2.7187, shapeName=dyn_chain_03_capsule, center=<2.1084 -0.0286 0.0036>, rotation=<rot 0 0.0003 -90>}]
 handle [{shapeType=Capsule, length=27.0878, radius=4.0037, center=<-0.0213 -0.0004 10.1416>, rotation=<rot 0.0146 -90 0.0171>}]
}
constraints [
 {bone1Name=dyn_chain_01, bone2Name=handle, swing1LimitDegrees=45, swing2LimitDegrees=45, twistLimitDegrees=1}
 {bone1Name=dyn_chain_02, bone2Name=dyn_chain_01, swing1LimitDegrees=45, swing2LimitDegrees=45, twistLimitDegrees=1}
 {bone1Name=dyn_chain_03, bone2Name=dyn_chain_02, swing1LimitDegrees=45, swing2LimitDegrees=45, twistLimitDegrees=1}
]
